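; Test variant of a farm boxes prop, a building actor Blueprint that by its name and folder is for physics testing.
; On BeginPlay it gets its physics object component through the BuildingActor physics simulation support and prints that component's display name to screen and log as a debug message.
; The construction script is empty, and the graph itself sets up or drives no physics.
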(event EventBeginPlay
  (bind _returnvalue (BuildingActor|PhysicsSimulationSupport|GetPhysicsObjectComponent self))
  (Development|PrintString (Utilities|GetDisplayName _returnvalue) true true "(R=0.000000,G=0.660000,B=1.000000,A=1.000000)" 2e+01))

(fn ConstructionScript)
